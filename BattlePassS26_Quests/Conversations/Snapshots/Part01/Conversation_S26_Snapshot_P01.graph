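# Conversation database for the Season 26 Snapshot Part 01 quests, holding NPC dialogue about recruiting a crew to rob Kado Thorne.
# Six tagged entry points lead into linear speech chains for Nolan Chance, Antonia, Agent Peely, Countess Daraku and Arctic Assassin.
# Opening lines auto-select when quest-state and NPC-tag requirements pass; each chain ends by completing a quest objective.
# Entry tag numbers differ from some quests: Q01 drives q02, and Q012.a/b/c drive q14.
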ConversationDatabase /BattlePassS26_Quests/Conversations/Snapshots/Part01/Conversation_S26_Snapshot_P01
entry_points [
 {tag=Conversation.Entry.Season26.Snapshot.P01.Q01, destinations [#1]}
 {tag=Conversation.Entry.Season26.Snapshot.P01.Q07, destinations [#2]}
 {tag=Conversation.Entry.Season26.Snapshot.P01.Q012.a, destinations [#3]}
 {tag=Conversation.Entry.Season26.Snapshot.P01.Q012.b, destinations [#4]}
 {tag=Conversation.Entry.Season26.Snapshot.P01.Q012.c, destinations [#5]}
 {tag=Conversation.Entry.Season26.Snapshot.P01.Q13, destinations [#6]}
]
nodes [
 ConversationEntryPointNode_5 {
  id=#6
  class=ConversationEntryPointNode
  outputs [#7]
  properties {entryTag {tagName=Conversation.Entry.Season26.Snapshot.P01.Q13}}
 }
 Speech_C_11 {
  id=#7
  class=Speech_C
  outputs [#8]
  properties {
   softTaskIcon=&T_UI_NPC_MainQuest
   generalConfig {defaultMessage="So, that job Antonia mentioned... It's a big one."}
  }
  sub_nodes [
   ConversationChoiceNode_0 {
    class=ConversationChoiceNode
    properties {
     defaultChoiceDisplayText=Talk to Nolan Chance
     choiceTags {gameplayTags [{tagName=Conversation.Choice.AutoSelectIfValid}, {tagName=Conversation.Choice.Category.MainQuestline}]}
    }
   }
   FortConversationRequirement_ControllerRequirement_0 {
    class=FortConversationRequirement_ControllerRequirement
    properties {
     nodeName=Require: Has Completed Quest Objective
     requirement {
      class=FortControllerRequirement_HasCompletedQuestObjective
      properties {
       bInverseCondition=true
       objectiveBackendName=quest_s26_snapshot_p01_q13_obj2
       questDefinition=/BattlePassS26_Quests/Items/QuestItems/Snapshot/Part01/quest_s26_snapshot_p01_q13
      }
     }
    }
   }
   FortConversationRequirement_ControllerRequirement_1 {
    class=FortConversationRequirement_ControllerRequirement
    properties {
     nodeName=Require: Is Quest Active
     requirement {
      class=FortControllerRequirement_IsQuestActive
      properties {questDefinition=&quest_s26_snapshot_p01_q13}
     }
    }
   }
   FortConversationRequirement_ControllerRequirement_3 {
    class=FortConversationRequirement_ControllerRequirement
    properties {
     nodeName=Require: NPC Has Conversation Tag
     participantId {tagName=Conversation.Participant.NPC}
     requirement {class=FortControllerRequirement_NPCHasConversationEntryTag, properties {queryTag {tagName=Conversation.Entry.MusketSlinger}}}
    }
   }
  ]
 }
 Speech_C_12 {
  id=#8
  class=Speech_C
  outputs [#9]
  properties {generalConfig {defaultMessage="There's something in Kado's deepest, best-protected vault that Slone wants."}}
  sub_nodes [ConversationChoiceNode_8 {class=ConversationChoiceNode, properties {defaultChoiceDisplayText=Continue}}]
 }
 Speech_C_14 {
  id=#9
  class=Speech_C
  outputs [#10]
  properties {generalConfig {defaultMessage="Pulling it off will require a crew stacked with the best of the best... But I've burned a lot of bridges. "}}
  sub_nodes [ConversationChoiceNode_10 {class=ConversationChoiceNode, properties {defaultChoiceDisplayText=Continue}}]
 }
 Speech_C_18 {
  id=#10
  class=Speech_C
  outputs [#11]
  properties {generalConfig {defaultMessage="If we can't recruit more new talent, like you... I'll have to make some VERY apologetic phone calls."}}
  sub_nodes [ConversationChoiceNode_12 {
    class=ConversationChoiceNode
    properties {defaultChoiceDisplayText=Continue, choiceTags {gameplayTags [{tagName=Athena.Quests.S26.Conversations.P01.Choice.Q13}]}}
   }]
 }
 FortConversationTaskNode_CompleteQuestObjectiveByBackendName_5 {
  id=#11
  class=FortConversationTaskNode_CompleteQuestObjectiveByBackendName
  properties {
   objectiveBackendName=quest_s26_snapshot_p01_q13_obj2
   questDefinition=&quest_s26_snapshot_p01_q13
  }
 }
 ConversationEntryPointNode_4 {
  id=#5
  class=ConversationEntryPointNode
  outputs [#12]
  properties {entryTag {tagName=Conversation.Entry.Season26.Snapshot.P01.Q012.c}}
 }
 Speech_C_8 {
  id=#12
  class=Speech_C
  outputs [#13]
  properties {
   softTaskIcon=&T_UI_NPC_MainQuest
   generalConfig {defaultMessage="Oh, wow... Uh... I'm really trying to rehabilitiate my image right now..."}
  }
  sub_nodes [
   ConversationChoiceNode_36 {
    class=ConversationChoiceNode
    properties {
     defaultChoiceDisplayText=Talk to Arctic Assassin
     choiceTags {gameplayTags [{tagName=Conversation.Choice.AutoSelectIfValid}, {tagName=Conversation.Choice.Category.MainQuestline}]}
    }
   }
   FortConversationRequirement_ControllerRequirement_68 {
    class=FortConversationRequirement_ControllerRequirement
    properties {
     nodeName=Require: Has Completed Quest
     requirement {
      class=FortControllerRequirement_HasCompletedQuest
      properties {bInverseCondition=true, questDefinition=/BattlePassS26_Quests/Items/QuestItems/Snapshot/Part01/quest_s26_snapshot_p01_q14}
     }
    }
   }
   FortConversationRequirement_ControllerRequirement_69 {
    class=FortConversationRequirement_ControllerRequirement
    properties {
     nodeName=Require: Is Quest Active
     requirement {
      class=FortControllerRequirement_IsQuestActive
      properties {questDefinition=&quest_s26_snapshot_p01_q14}
     }
    }
   }
   FortConversationRequirement_ControllerRequirement_6 {
    class=FortConversationRequirement_ControllerRequirement
    properties {
     nodeName=Require: Has Completed Quest Objective
     requirement {
      class=FortControllerRequirement_HasCompletedQuestObjective
      properties {
       bInverseCondition=true
       objectiveBackendName=quest_s26_snapshot_p01_q14_obj2
       questDefinition=&quest_s26_snapshot_p01_q14
      }
     }
    }
   }
   FortConversationRequirement_ControllerRequirement_71 {
    class=FortConversationRequirement_ControllerRequirement
    properties {
     nodeName=Require: NPC Has Conversation Tag
     participantId {tagName=Conversation.Participant.NPC}
     requirement {class=FortControllerRequirement_NPCHasConversationEntryTag, properties {queryTag {tagName=Conversation.Entry.ArcticAssassin}}}
    }
   }
  ]
 }
 Speech_C_9 {
  id=#13
  class=Speech_C
  outputs [#14]
  properties {generalConfig {defaultMessage="Which is already hard enough with a last name like \"assassin.\" "}}
  sub_nodes [ConversationChoiceNode_17 {
    class=ConversationChoiceNode
    properties {
     defaultChoiceDisplayText=Continue
     choiceTags {gameplayTags [{tagName=Athena.Quests.S26.Conversations.P01.Choice.Q14.C}, {tagName=Athena.Quests.S26.Conversations.P01.Choice}]}
    }
   }]
 }
 Speech_C_17 {
  id=#14
  class=Speech_C
  outputs [#15]
  properties {generalConfig {defaultMessage="But, um, good luck with your heist!"}}
  sub_nodes [ConversationChoiceNode_16 {
    class=ConversationChoiceNode
    properties {
     defaultChoiceDisplayText=Continue
     choiceTags {gameplayTags [{tagName=Athena.Quests.S26.Conversations.P01.Choice.Q14.C}, {tagName=Athena.Quests.S26.Conversations.P01.Choice}]}
    }
   }]
 }
 FortConversationTaskNode_CompleteQuestObjectiveByBackendName_4 {
  id=#15
  class=FortConversationTaskNode_CompleteQuestObjectiveByBackendName
  properties {
   objectiveBackendName=quest_s26_snapshot_p01_q14_obj2
   questDefinition=&quest_s26_snapshot_p01_q14
  }
 }
 ConversationEntryPointNode_3 {
  id=#4
  class=ConversationEntryPointNode
  outputs [#16]
  properties {entryTag {tagName=Conversation.Entry.Season26.Snapshot.P01.Q012.b}}
 }
 Speech_C_6 {
  id=#16
  class=Speech_C
  outputs [#17]
  properties {
   softTaskIcon=&T_UI_NPC_MainQuest
   generalConfig {defaultMessage="Me? Not to be totally cliche, but... You know who I am, right?"}
  }
  sub_nodes [
   ConversationChoiceNode_6 {
    class=ConversationChoiceNode
    properties {
     defaultChoiceDisplayText=Talk to Countess Daraku
     choiceTags {gameplayTags [{tagName=Conversation.Choice.AutoSelectIfValid}, {tagName=Conversation.Choice.Category.MainQuestline}]}
    }
   }
   FortConversationRequirement_ControllerRequirement_11 {
    class=FortConversationRequirement_ControllerRequirement
    properties {
     nodeName=Require: Has Completed Quest
     requirement {
      class=FortControllerRequirement_HasCompletedQuest
      properties {bInverseCondition=true, questDefinition=&quest_s26_snapshot_p01_q14}
     }
    }
   }
   FortConversationRequirement_ControllerRequirement_12 {
    class=FortConversationRequirement_ControllerRequirement
    properties {
     nodeName=Require: Is Quest Active
     requirement {
      class=FortControllerRequirement_IsQuestActive
      properties {questDefinition=&quest_s26_snapshot_p01_q14}
     }
    }
   }
   FortConversationRequirement_ControllerRequirement_13 {
    class=FortConversationRequirement_ControllerRequirement
    properties {
     nodeName=Require: Has Completed Quest Objective
     requirement {
      class=FortControllerRequirement_HasCompletedQuestObjective
      properties {
       bInverseCondition=true
       objectiveBackendName=quest_s26_snapshot_p01_q14_obj1
       questDefinition=&quest_s26_snapshot_p01_q14
      }
     }
    }
   }
   FortConversationRequirement_ControllerRequirement_14 {
    class=FortConversationRequirement_ControllerRequirement
    properties {
     nodeName=Require: NPC Has Conversation Tag
     participantId {tagName=Conversation.Participant.NPC}
     requirement {class=FortControllerRequirement_NPCHasConversationEntryTag, properties {queryTag {tagName=Conversation.Entry.IchorIncisor}}}
    }
   }
  ]
 }
 Speech_C_7 {
  id=#17
  class=Speech_C
  outputs [#18]
  properties {generalConfig {defaultMessage=I'm not really the 'low profile' type. I draw crowds. Big ones.}}
  sub_nodes [ConversationChoiceNode_7 {
    class=ConversationChoiceNode
    properties {
     defaultChoiceDisplayText=Continue
     choiceTags {gameplayTags [{tagName=Athena.Quests.S26.Conversations.P01.Choice.Q14.B}, {tagName=Athena.Quests.S26.Conversations.P01.Choice}]}
    }
   }]
 }
 FortConversationTaskNode_CompleteQuestObjectiveByBackendName_3 {
  id=#18
  class=FortConversationTaskNode_CompleteQuestObjectiveByBackendName
  properties {
   objectiveBackendName=quest_s26_snapshot_p01_q14_obj1
   questDefinition=&quest_s26_snapshot_p01_q14
  }
 }
 ConversationEntryPointNode_2 {
  id=#3
  class=ConversationEntryPointNode
  outputs [#19]
  properties {entryTag {tagName=Conversation.Entry.Season26.Snapshot.P01.Q012.a}}
 }
 Speech_C_4 {
  id=#19
  class=Speech_C
  outputs [#20]
  properties {softTaskIcon=&T_UI_NPC_MainQuest, generalConfig {defaultMessage=(Agent Peely frowns and shakes his head.)}}
  sub_nodes [
   ConversationChoiceNode_4 {
    class=ConversationChoiceNode
    properties {
     defaultChoiceDisplayText=Talk to Agent Peely
     choiceTags {gameplayTags [{tagName=Conversation.Choice.AutoSelectIfValid}, {tagName=Conversation.Choice.Category.MainQuestline}]}
    }
   }
   FortConversationRequirement_ControllerRequirement_7 {
    class=FortConversationRequirement_ControllerRequirement
    properties {
     nodeName=Require: Has Completed Quest
     requirement {
      class=FortControllerRequirement_HasCompletedQuest
      properties {bInverseCondition=true, questDefinition=&quest_s26_snapshot_p01_q14}
     }
    }
   }
   FortConversationRequirement_ControllerRequirement_8 {
    class=FortConversationRequirement_ControllerRequirement
    properties {
     nodeName=Require: Is Quest Active
     requirement {
      class=FortControllerRequirement_IsQuestActive
      properties {questDefinition=&quest_s26_snapshot_p01_q14}
     }
    }
   }
   FortConversationRequirement_ControllerRequirement_9 {
    class=FortConversationRequirement_ControllerRequirement
    properties {
     nodeName=Require: Has Completed Quest Objective
     requirement {
      class=FortControllerRequirement_HasCompletedQuestObjective
      properties {
       bInverseCondition=true
       objectiveBackendName=quest_s26_snapshot_p01_q14_obj0
       questDefinition=&quest_s26_snapshot_p01_q14
      }
     }
    }
   }
   FortConversationRequirement_ControllerRequirement_10 {
    class=FortConversationRequirement_ControllerRequirement
    properties {
     nodeName=Require: NPC Has Conversation Tag
     participantId {tagName=Conversation.Participant.NPC}
     requirement {class=FortControllerRequirement_NPCHasConversationEntryTag, properties {queryTag {tagName=Conversation.Entry.AgentPeely}}}
    }
   }
  ]
 }
 Speech_C_5 {
  id=#20
  class=Speech_C
  outputs [#21]
  properties {generalConfig {defaultMessage=(He looks disappointed that you have blown his cover.)}}
  sub_nodes [ConversationChoiceNode_5 {
    class=ConversationChoiceNode
    properties {
     defaultChoiceDisplayText=Continue
     choiceTags {gameplayTags [{tagName=Athena.Quests.S26.Conversations.P01.Choice.Q14.A}, {tagName=Athena.Quests.S26.Conversations.P01.Choice}]}
    }
   }]
 }
 FortConversationTaskNode_CompleteQuestObjectiveByBackendName_1 {
  id=#21
  class=FortConversationTaskNode_CompleteQuestObjectiveByBackendName
  properties {
   objectiveBackendName=quest_s26_snapshot_p01_q14_obj0
   questDefinition=&quest_s26_snapshot_p01_q14
  }
 }
 ConversationEntryPointNode_0 {
  id=#2
  class=ConversationEntryPointNode
  outputs [#22]
  properties {entryTag {tagName=Conversation.Entry.Season26.Snapshot.P01.Q07}}
 }
 Speech_C_0 {
  id=#22
  class=Speech_C
  outputs [#23]
  properties {
   softTaskIcon=&T_UI_NPC_MainQuest
   generalConfig {defaultMessage="Nolan and I are partners. Co-conspirators. A real one-two punch. Or, were, I suppose. "}
  }
  sub_nodes [
   ConversationChoiceNode_2 {
    class=ConversationChoiceNode
    properties {
     defaultChoiceDisplayText=Talk to Antonia
     choiceTags {gameplayTags [{tagName=Conversation.Choice.AutoSelectIfValid}, {tagName=Conversation.Choice.Category.MainQuestline}]}
    }
   }
   FortConversationRequirement_ControllerRequirement_2 {
    class=FortConversationRequirement_ControllerRequirement
    properties {
     nodeName=Require: Has Completed Quest
     requirement {
      class=FortControllerRequirement_HasCompletedQuest
      properties {bInverseCondition=true, questDefinition=/BattlePassS26_Quests/Items/QuestItems/Snapshot/Part01/quest_s26_snapshot_p01_q7}
     }
    }
   }
   FortConversationRequirement_ControllerRequirement_4 {
    class=FortConversationRequirement_ControllerRequirement
    properties {
     nodeName=Require: Is Quest Active
     requirement {
      class=FortControllerRequirement_IsQuestActive
      properties {questDefinition=&quest_s26_snapshot_p01_q7}
     }
    }
   }
   FortConversationRequirement_ControllerRequirement_5 {
    class=FortConversationRequirement_ControllerRequirement
    properties {
     nodeName=Require: NPC Has Conversation Tag
     participantId {tagName=Conversation.Participant.NPC}
     requirement {class=FortControllerRequirement_NPCHasConversationEntryTag, properties {queryTag {tagName=Conversation.Entry.BlazerVeil}}}
    }
   }
  ]
 }
 Speech_C_3 {
  id=#23
  class=Speech_C
  outputs [#24]
  properties {generalConfig {defaultMessage="On our last job, he bailed with our driver and left me holding the bag."}}
  sub_nodes [ConversationChoiceNode_14 {class=ConversationChoiceNode, properties {defaultChoiceDisplayText=Continue}}]
 }
 Speech_C_15 {
  id=#24
  class=Speech_C
  outputs [#25]
  properties {generalConfig {defaultMessage="I got out. I always get out. But he never apologized, so... I'm stealing this job from him."}}
  sub_nodes [ConversationChoiceNode_3 {class=ConversationChoiceNode, properties {defaultChoiceDisplayText=Continue}}]
 }
 Speech_C_16 {
  id=#25
  class=Speech_C
  outputs [#26]
  properties {generalConfig {defaultMessage=Are you in?}}
  sub_nodes [ConversationChoiceNode_15 {
    class=ConversationChoiceNode
    properties {defaultChoiceDisplayText=Continue, choiceTags {gameplayTags [{tagName=Athena.Quests.S26.Conversations.P01.Choice.Q07}]}}
   }]
 }
 FortConversationTaskNode_CompleteQuestObjectiveByBackendName_0 {
  id=#26
  class=FortConversationTaskNode_CompleteQuestObjectiveByBackendName
  properties {
   objectiveBackendName=quest_s26_snapshot_p01_q7_obj0
   questDefinition=&quest_s26_snapshot_p01_q7
  }
 }
 ConversationEntryPointNode_1 {
  id=#1
  class=ConversationEntryPointNode
  outputs [#27]
  properties {entryTag {tagName=Conversation.Entry.Season26.Snapshot.P01.Q01}}
 }
 Speech_C_1 {
  id=#27
  class=Speech_C
  outputs [#28]
  properties {
   softTaskIcon=&T_UI_NPC_MainQuest
   generalConfig {defaultMessage="A mutual friend told me you'd be perfect for a heisting gig. You know Slone, right?"}
  }
  sub_nodes [
   ConversationChoiceNode_1 {
    class=ConversationChoiceNode
    properties {
     defaultChoiceDisplayText=Talk to Nolan Chance
     choiceTags {gameplayTags [{tagName=Conversation.Choice.AutoSelectIfValid}, {tagName=Conversation.Choice.Category.MainQuestline}]}
    }
   }
   FortConversationRequirement_ControllerRequirement_15 {
    class=FortConversationRequirement_ControllerRequirement
    properties {
     nodeName=Require: Has Completed Quest
     requirement {
      class=FortControllerRequirement_HasCompletedQuest
      properties {bInverseCondition=true, questDefinition=/BattlePassS26_Quests/Items/QuestItems/Snapshot/Part01/quest_s26_snapshot_p01_q02}
     }
    }
   }
   FortConversationRequirement_ControllerRequirement_16 {
    class=FortConversationRequirement_ControllerRequirement
    properties {
     nodeName=Require: Is Quest Active
     requirement {
      class=FortControllerRequirement_IsQuestActive
      properties {questDefinition=&quest_s26_snapshot_p01_q02}
     }
    }
   }
   FortConversationRequirement_ControllerRequirement_17 {
    class=FortConversationRequirement_ControllerRequirement
    properties {
     nodeName=Require: NPC Has Conversation Tag
     participantId {tagName=Conversation.Participant.NPC}
     requirement {class=FortControllerRequirement_NPCHasConversationEntryTag, properties {queryTag {tagName=Conversation.Entry.MusketSlinger}}}
    }
   }
  ]
 }
 Speech_C_2 {
  id=#28
  class=Speech_C
  outputs [#29]
  properties {generalConfig {defaultMessage="Our target is Kado Thorne, and he has MAJOR security. Turrets, cameras, guards..."}}
  sub_nodes [ConversationChoiceNode_9 {class=ConversationChoiceNode, properties {defaultChoiceDisplayText=Continue}}]
 }
 Speech_C_10 {
  id=#29
  class=Speech_C
  outputs [#30]
  properties {generalConfig {defaultMessage="A lot of guards. So, so many guards. He's definitely guarding something important down there..."}}
  sub_nodes [ConversationChoiceNode_11 {class=ConversationChoiceNode, properties {defaultChoiceDisplayText=Continue}}]
 }
 Speech_C_13 {
  id=#30
  class=Speech_C
  outputs [#31]
  properties {generalConfig {defaultMessage=Think you can handle the turrets and cameras? Explosives should do the trick.}}
  sub_nodes [ConversationChoiceNode_13 {
    class=ConversationChoiceNode
    properties {defaultChoiceDisplayText=Continue, choiceTags {gameplayTags [{tagName=Athena.Quests.S26.Conversations.P01.Choice.Q02}]}}
   }]
 }
 FortConversationTaskNode_CompleteQuestObjectiveByBackendName_2 {
  id=#31
  class=FortConversationTaskNode_CompleteQuestObjectiveByBackendName
  properties {
   objectiveBackendName=quest_s26_snapshot_p01_q02_obj0
   questDefinition=&quest_s26_snapshot_p01_q02
  }
 }
]
